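# Season 16 foreshadowing conversation for a helicopter quest prop, staged as tuning a static-filled radio to recover a military transmission.
# Unlabeled plus and minus "Tune" choices branch from a partly garbled opening: plus then minus reaches the clear message and fires the success side effect.
# Successive minus choices degrade the message to pure hiss, while other plus choices return to the start.
# Speech steps send numbered signal-clarity messages to the quest prop actor.
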ConversationDatabase /BattlepassS16/QuestActors/Conversations/Conversation_Foreshadowing_Helicopter
entry_points [{tag=Conversation.Entry.Season16.Foreshadowing.Helicopter, destinations [#1]}]
nodes [
 ConversationEntryPointNode_7 {
  id=#1
  class=ConversationEntryPointNode
  outputs [#2]
  properties {entryTag {tagName=Conversation.Entry.Season16.Foreshadowing.Helicopter}}
 }
 Speech_C_0 {
  id=#2
  class=Speech_C
  outputs [#3, #4]
  properties {generalConfig {defaultMessage="*hiss* truck inbound, OPSEC level 3 *hiss*"}}
  sub_nodes [ConvoSideEffect_QuestPropActor_Message_C_0 {class=ConvoSideEffect_QuestPropActor_Message_C, properties {message Number=3}}]
 }
 Speech_C_11 {
  id=#4
  class=Speech_C
  outputs [#5, #6]
  properties {
   softTaskIcon=/Game/UI/Foundation/Textures/Icons/Replay/T-Icon-Replay-Plus-L
   generalConfig {defaultMessage="*hiss* field decon truck inbound, OPSEC level 3 is in effect *hiss*"}
  }
  sub_nodes [
   ConversationChoiceNode_6 {class=ConversationChoiceNode, properties {defaultChoiceDisplayText=Tune}}
   ConvoSideEffect_QuestPropActor_Message_C_3 {class=ConvoSideEffect_QuestPropActor_Message_C, properties {message Number=4}}
  ]
 }
 ReturnToStart_C_0 {
  id=#6
  class=ReturnToStart_C
  properties {softTaskIcon=&T-Icon-Replay-Plus-L}
  sub_nodes [ConversationChoiceNode_9 {class=ConversationChoiceNode, properties {defaultChoiceDisplayText=Tune}}]
 }
 Speech_C_3 {
  id=#5
  class=Speech_C
  outputs []
  properties {
   softTaskIcon=/Game/UI/Foundation/Textures/Icons/Replay/T-Icon-Replay-Minus-L
   generalConfig {defaultMessage="<b>India Oscar</> field decon truck inbound, <enemy>OPSEC level 3</> is in effect, repeat, OPSEC 3 now a GO."}
  }
  sub_nodes [
   ConversationChoiceNode_5 {class=ConversationChoiceNode, properties {defaultChoiceDisplayText=Tune}}
   ConvoSideEffect_QuestPropActor_ConvoSuccess_C_1 {class=ConvoSideEffect_QuestPropActor_ConvoSuccess_C}
   ConvoSideEffect_QuestPropActor_Message_C_5 {class=ConvoSideEffect_QuestPropActor_Message_C, properties {message Number=5}}
  ]
 }
 Speech_C_9 {
  id=#3
  class=Speech_C
  outputs [#7, #6]
  properties {softTaskIcon=&T-Icon-Replay-Minus-L, generalConfig {defaultMessage=*hiss* truck inbound *hiss*}}
  sub_nodes [
   ConversationChoiceNode_10 {class=ConversationChoiceNode, properties {defaultChoiceDisplayText=Tune}}
   ConvoSideEffect_QuestPropActor_Message_C_1 {class=ConvoSideEffect_QuestPropActor_Message_C, properties {message Number=2}}
  ]
 }
 Speech_C_1 {
  id=#7
  class=Speech_C
  outputs [#8, #6]
  properties {softTaskIcon=&T-Icon-Replay-Minus-L, generalConfig {defaultMessage=*hiss*}}
  sub_nodes [
   ConversationChoiceNode_3 {class=ConversationChoiceNode, properties {defaultChoiceDisplayText=Tune}}
   ConvoSideEffect_QuestPropActor_Message_C_4 {class=ConvoSideEffect_QuestPropActor_Message_C, properties {message Number=1}}
  ]
 }
 Back_C_0 {
  id=#8
  class=Back_C
  properties {softTaskIcon=&T-Icon-Replay-Minus-L}
  sub_nodes [
   ConvoSideEffect_QuestPropActor_Message_C_2 {class=ConvoSideEffect_QuestPropActor_Message_C, properties {message Number=1}}
   ConversationChoiceNode_7 {class=ConversationChoiceNode, properties {defaultChoiceDisplayText=Tune}}
  ]
 }
]
