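# Conversation database for the CubeNinja NPC's dialogue in the Season 16 Spire main questline, covering quest stages q05 through q11.
# One entry point fans out to "The Spire" speech branches, each gated by quest active, turn-in or completed requirements.
# Progress branches turn in quests and grant the next, some behind Accept/Back choices; others give reminders.
# The three-part q08 fetch stage uses several partial-progress branches sharing one reminder chain.
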ConversationDatabase /BattlepassS16/QuestActors/Conversations/Shared_Conversation_TheSpire_CubeNinja
entry_points [{tag=Conversation.Entry.Season16.TheSpire.CubeNinja, destinations [#1]}]
nodes [
 ConversationEntryPointNode_7 {
  id=#1
  class=ConversationEntryPointNode
  outputs [
   #2
   #3
   #4
   #5
   #6
   #7
   #8
   #9
   #10
   #11
   #12
   #13
   #14
  ]
  properties {entryTag {tagName=Conversation.Entry.Season16.TheSpire.CubeNinja}}
 }
 Speech_C_13 {
  id=#14
  class=Speech_C
  outputs [#15]
  properties {
   softTaskIcon=&T_UI_NPC_MainQuest
   generalConfig {defaultMessage="My work cannot be rushed, and the <b>Old Guardians</> must have their dues. "}
  }
  sub_nodes [
   ConversationChoiceNode_5 {
    class=ConversationChoiceNode
    properties {defaultChoiceDisplayText=The Spire, choiceTags {gameplayTags [{tagName=Conversation.Choice.Category.MainQuestline}]}}
   }
   FortConversationRequirement_ControllerRequirement_0 {
    class=FortConversationRequirement_ControllerRequirement
    properties {
     nodeName=Require: Is Quest Active
     requirement {class=FortControllerRequirement_IsQuestActive, properties {questDefinition=/BattlepassS16/Items/QuestItems/Narrative/quest_s16_tower_q08a}}
    }
   }
   FortConversationRequirement_ControllerRequirement_11 {
    class=FortConversationRequirement_ControllerRequirement
    properties {
     nodeName=Require: Has Quest To Turn In
     requirement {
      class=FortControllerRequirement_HasQuestToTurnIn
      properties {softQuestToTurnIn=&quest_s16_tower_q08a}
     }
    }
   }
   FortConversationRequirement_ControllerRequirement_38 {
    class=FortConversationRequirement_ControllerRequirement
    properties {
     nodeName=Require: Has Quest To Turn In
     requirement {
      class=FortControllerRequirement_HasQuestToTurnIn
      properties {softQuestToTurnIn=/BattlepassS16/Items/QuestItems/Narrative/quest_s16_tower_q08b}
     }
    }
   }
   FortConversationRequirement_ControllerRequirement_39 {
    class=FortConversationRequirement_ControllerRequirement
    properties {
     nodeName=Require: Has Quest To Turn In
     requirement {
      class=FortControllerRequirement_HasQuestToTurnIn
      properties {bInverseCondition=true, softQuestToTurnIn=/BattlepassS16/Items/QuestItems/Narrative/quest_s16_tower_q08c}
     }
    }
   }
   FortConversationRequirement_ControllerRequirement_19 {
    class=FortConversationRequirement_ControllerRequirement
    properties {
     nodeName=Require: Has Completed Quest
     requirement {
      class=FortControllerRequirement_HasCompletedQuest
      properties {bInverseCondition=true, questDefinition=&quest_s16_tower_q08a}
     }
    }
   }
  ]
 }
 Speech_C_42 {
  id=#15
  class=Speech_C
  outputs [#16]
  properties {generalConfig {defaultMessage="Bring me an <b>artifact stolen</>from the Spire, a Spire <enemy>Talisman looted</> from a Spire Guardian's body..."}}
 }
 Speech_C_43 {
  id=#16
  class=Speech_C
  outputs [#17]
  properties {generalConfig {defaultMessage="A Wolf Fang, Boar Tusk, and three Chicken Feathers. "}}
 }
 Speech_C_44 {id=#17, class=Speech_C, outputs [], properties {generalConfig {defaultMessage=And then leave me to my work.}}}
 Speech_C_12 {
  id=#13
  class=Speech_C
  outputs [#15]
  properties {
   softTaskIcon=&T_UI_NPC_MainQuest
   generalConfig {defaultMessage="My work cannot be rushed, and the <b>Old Guardians</> must have their dues. "}
  }
  sub_nodes [
   ConversationChoiceNode_13 {
    class=ConversationChoiceNode
    properties {defaultChoiceDisplayText=The Spire, choiceTags {gameplayTags [{tagName=Conversation.Choice.Category.MainQuestline}]}}
   }
   FortConversationRequirement_ControllerRequirement_32 {
    class=FortConversationRequirement_ControllerRequirement
    properties {
     nodeName=Require: Is Quest Active
     requirement {class=FortControllerRequirement_IsQuestActive, properties {questDefinition=&quest_s16_tower_q08a}}
    }
   }
   FortConversationRequirement_ControllerRequirement_36 {
    class=FortConversationRequirement_ControllerRequirement
    properties {
     nodeName=Require: Has Quest To Turn In
     requirement {
      class=FortControllerRequirement_HasQuestToTurnIn
      properties {softQuestToTurnIn=&quest_s16_tower_q08a}
     }
    }
   }
   FortConversationRequirement_ControllerRequirement_37 {
    class=FortConversationRequirement_ControllerRequirement
    properties {
     nodeName=Require: Has Quest To Turn In
     requirement {
      class=FortControllerRequirement_HasQuestToTurnIn
      properties {bInverseCondition=true, softQuestToTurnIn=&quest_s16_tower_q08b}
     }
    }
   }
   FortConversationRequirement_ControllerRequirement_18 {
    class=FortConversationRequirement_ControllerRequirement
    properties {
     nodeName=Require: Has Completed Quest
     requirement {
      class=FortControllerRequirement_HasCompletedQuest
      properties {bInverseCondition=true, questDefinition=&quest_s16_tower_q08a}
     }
    }
   }
  ]
 }
 Speech_C_10 {
  id=#12
  class=Speech_C
  outputs []
  properties {softTaskIcon=&T_UI_NPC_MainQuest, generalConfig {defaultMessage=I need more time. Come back later.}}
  sub_nodes [
   ConversationChoiceNode_2 {
    class=ConversationChoiceNode
    properties {defaultChoiceDisplayText=The Spire, choiceTags {gameplayTags [{tagName=Conversation.Choice.Category.MainQuestline}]}}
   }
   FortConversationRequirement_ControllerRequirement_8 {
    class=FortConversationRequirement_ControllerRequirement
    properties {
     nodeName=Require: Has Completed Quest
     requirement {
      class=FortControllerRequirement_HasCompletedQuest
      properties {bInverseCondition=true, questDefinition=/BattlepassS16/Items/QuestItems/Narrative/quest_s16_tower_q07}
     }
    }
   }
   FortConversationRequirement_ControllerRequirement_9 {
    class=FortConversationRequirement_ControllerRequirement
    properties {
     nodeName=Require: Has Quest To Turn In
     requirement {
      class=FortControllerRequirement_HasQuestToTurnIn
      properties {bInverseCondition=true, softQuestToTurnIn=&quest_s16_tower_q07}
     }
    }
   }
   FortConversationRequirement_ControllerRequirement_15 {
    class=FortConversationRequirement_ControllerRequirement
    properties {
     nodeName=Require: Has Completed Quest
     requirement {
      class=FortControllerRequirement_HasCompletedQuest
      properties {questDefinition=/BattlepassS16/Items/QuestItems/Narrative/quest_s16_tower_q06}
     }
    }
   }
  ]
 }
 Speech_C_15 {
  id=#11
  class=Speech_C
  outputs [#18]
  properties {softTaskIcon=&T_UI_NPC_MainQuest, generalConfig {defaultMessage="Patience. "}}
  sub_nodes [
   ConversationChoiceNode_16 {
    class=ConversationChoiceNode
    properties {defaultChoiceDisplayText=The Spire, choiceTags {gameplayTags [{tagName=Conversation.Choice.Category.MainQuestline}]}}
   }
   FortConversationRequirement_ControllerRequirement_20 {
    class=FortConversationRequirement_ControllerRequirement
    properties {
     nodeName=Require: Has Completed Quest
     requirement {
      class=FortControllerRequirement_HasCompletedQuest
      properties {bInverseCondition=true, questDefinition=/BattlepassS16/Items/QuestItems/Narrative/quest_s16_tower_q09}
     }
    }
   }
   FortConversationRequirement_ControllerRequirement_3 {
    class=FortConversationRequirement_ControllerRequirement
    properties {
     nodeName=Require: Has Quest To Turn In
     requirement {
      class=FortControllerRequirement_HasQuestToTurnIn
      properties {bInverseCondition=true, softQuestToTurnIn=&quest_s16_tower_q09}
     }
    }
   }
   FortConversationRequirement_ControllerRequirement_24 {
    class=FortConversationRequirement_ControllerRequirement
    properties {
     nodeName=Require: Is Quest Active
     requirement {class=FortControllerRequirement_IsQuestActive, properties {questDefinition=&quest_s16_tower_q09}}
    }
   }
   FortConversationRequirement_ControllerRequirement_31 {
    class=FortConversationRequirement_ControllerRequirement
    properties {
     nodeName=Require: Has Completed Quest
     requirement {
      class=FortControllerRequirement_HasCompletedQuest
      properties {questDefinition=&quest_s16_tower_q08a}
     }
    }
   }
  ]
 }
 Speech_C_53 {
  id=#18
  class=Speech_C
  outputs []
  properties {softTaskIcon=&T_UI_NPC_MainQuest, generalConfig {defaultMessage="The deed is not done. "}}
 }
 Speech_C_35 {
  id=#10
  class=Speech_C
  outputs [#19]
  properties {
   softTaskIcon=&T_UI_NPC_MainQuest
   generalConfig {defaultMessage="This message of yours is directing you to <b>the Spire. </>"}
  }
  sub_nodes [
   ConversationChoiceNode_8 {
    class=ConversationChoiceNode
    properties {defaultChoiceDisplayText=The Spire, choiceTags {gameplayTags [{tagName=Conversation.Choice.Category.MainQuestline}]}}
   }
   FortConversationRequirement_ControllerRequirement_5 {
    class=FortConversationRequirement_ControllerRequirement
    properties {
     nodeName=Require: Has Completed Quest
     requirement {
      class=FortControllerRequirement_HasCompletedQuest
      properties {bInverseCondition=true, questDefinition=&quest_s16_tower_q07}
     }
    }
   }
   FortConversationRequirement_ControllerRequirement_6 {
    class=FortConversationRequirement_ControllerRequirement
    properties {
     nodeName=Require: Has Quest To Turn In
     requirement {
      class=FortControllerRequirement_HasQuestToTurnIn
      properties {softQuestToTurnIn=&quest_s16_tower_q07}
     }
    }
   }
   FortConversationRequirement_ControllerRequirement_16 {
    class=FortConversationRequirement_ControllerRequirement
    properties {
     nodeName=Require: Is Quest Active
     requirement {class=FortControllerRequirement_IsQuestActive, properties {questDefinition=&quest_s16_tower_q07}}
    }
   }
   FortConversationRequirement_ControllerRequirement_27 {
    class=FortConversationRequirement_ControllerRequirement
    properties {
     nodeName=Require: Is Quest Active
     requirement {
      class=FortControllerRequirement_IsQuestActive
      properties {bInverseCondition=true, questDefinition=&quest_s16_tower_q08a}
     }
    }
   }
   FortConversationRequirement_ControllerRequirement_28 {
    class=FortConversationRequirement_ControllerRequirement
    properties {
     nodeName=Require: Has Completed Quest
     requirement {
      class=FortControllerRequirement_HasCompletedQuest
      properties {bInverseCondition=true, questDefinition=&quest_s16_tower_q08a}
     }
    }
   }
  ]
 }
 Speech_C_36 {
  id=#19
  class=Speech_C
  outputs [#20]
  properties {
   softTaskIcon=&T_UI_NPC_MainQuest
   generalConfig {defaultMessage="Coincidentally, this is a place of great interest to me. "}
  }
 }
 Speech_C_3 {
  id=#20
  class=Speech_C
  outputs [#21]
  properties {generalConfig {defaultMessage="I have been working on a way to disguise myself to finally enter the <b>Spire's inner sanctum. </>"}}
 }
 Speech_C_6 {
  id=#21
  class=Speech_C
  outputs [#22, #23]
  properties {generalConfig {defaultMessage="I can make a disguise for you as well... if you bring me what I require. "}}
 }
 TurnInQuest_C_6 {
  id=#23
  class=TurnInQuest_C
  outputs [#24]
  properties {softTaskIcon=/Game/Athena/HUD/Interaction/T_UI_NPC_Confirm}
  sub_nodes [ConversationChoiceNode_26 {class=ConversationChoiceNode, properties {defaultChoiceDisplayText=Accept}}]
 }
 GrantStaticQuest_C_6 {
  id=#24
  class=GrantStaticQuest_C
  outputs [#25]
  properties {questToGrantPtr=&quest_s16_tower_q08a}
 }
 GrantStaticQuest_C_7 {
  id=#25
  class=GrantStaticQuest_C
  outputs [#26]
  properties {questToGrantPtr=&quest_s16_tower_q08b}
 }
 GrantStaticQuest_C_8 {
  id=#26
  class=GrantStaticQuest_C
  outputs [#27]
  properties {questToGrantPtr=&quest_s16_tower_q08c}
 }
 Speech_C_52 {
  id=#27
  class=Speech_C
  outputs [#28]
  properties {generalConfig {defaultMessage="Bring me an artifact <b>stolen</> from the Spire. Mind you, it <b>must</> be stolen. "}}
 }
 Speech_C_38 {
  id=#28
  class=Speech_C
  outputs [#29]
  properties {generalConfig {defaultMessage="A talisman looted from a <enemy>Spire Guardian's body...</>"}}
 }
 Speech_C_39 {
  id=#29
  class=Speech_C
  outputs [#30]
  properties {generalConfig {defaultMessage="And a sacrifice from the wilds in the form of a <b>Wolf Fang, a Boar Tusk, and three Chicken Feathers. </>"}}
 }
 Speech_C_40 {
  id=#30
  class=Speech_C
  outputs []
  properties {generalConfig {defaultMessage="Bring me this and the <b>Old Guardians</> may just bless our endeavors. "}}
 }
 Back_C_2 {
  id=#22
  class=Back_C
  sub_nodes [ConversationChoiceNode_25 {class=ConversationChoiceNode, properties {defaultChoiceDisplayText=BACK}}]
 }
 Speech_C_41 {
  id=#9
  class=Speech_C
  outputs [#15]
  properties {
   softTaskIcon=&T_UI_NPC_MainQuest
   generalConfig {defaultMessage="My work cannot be rushed, and the <b>Old Guardians</> must have their dues. "}
  }
  sub_nodes [
   ConversationChoiceNode_12 {
    class=ConversationChoiceNode
    properties {defaultChoiceDisplayText=The Spire, choiceTags {gameplayTags [{tagName=Conversation.Choice.Category.MainQuestline}]}}
   }
   FortConversationRequirement_ControllerRequirement_29 {
    class=FortConversationRequirement_ControllerRequirement
    properties {
     nodeName=Require: Has Completed Quest
     requirement {
      class=FortControllerRequirement_HasCompletedQuest
      properties {bInverseCondition=true, questDefinition=&quest_s16_tower_q08a}
     }
    }
   }
   FortConversationRequirement_ControllerRequirement_30 {
    class=FortConversationRequirement_ControllerRequirement
    properties {
     nodeName=Require: Has Quest To Turn In
     requirement {
      class=FortControllerRequirement_HasQuestToTurnIn
      properties {bInverseCondition=true, softQuestToTurnIn=&quest_s16_tower_q08a}
     }
    }
   }
   FortConversationRequirement_ControllerRequirement_17 {
    class=FortConversationRequirement_ControllerRequirement
    properties {
     nodeName=Require: Is Quest Active
     requirement {class=FortControllerRequirement_IsQuestActive, properties {questDefinition=&quest_s16_tower_q08a}}
    }
   }
  ]
 }
 Speech_C_34 {
  id=#8
  class=Speech_C
  outputs [#31]
  properties {softTaskIcon=&T_UI_NPC_MainQuest, generalConfig {defaultMessage="Interesting message, no?"}}
  sub_nodes [
   ConversationChoiceNode_4 {
    class=ConversationChoiceNode
    properties {defaultChoiceDisplayText=The Spire, choiceTags {gameplayTags [{tagName=Conversation.Choice.Category.MainQuestline}]}}
   }
   FortConversationRequirement_ControllerRequirement_4 {
    class=FortConversationRequirement_ControllerRequirement
    properties {
     nodeName=Require: Has Completed Quest
     requirement {
      class=FortControllerRequirement_HasCompletedQuest
      properties {bInverseCondition=true, questDefinition=&quest_s16_tower_q06}
     }
    }
   }
   FortConversationRequirement_ControllerRequirement_14 {
    class=FortConversationRequirement_ControllerRequirement
    properties {
     nodeName=Require: Has Quest To Turn In
     requirement {
      class=FortControllerRequirement_HasQuestToTurnIn
      properties {softQuestToTurnIn=&quest_s16_tower_q06}
     }
    }
   }
  ]
 }
 Speech_C_33 {
  id=#31
  class=Speech_C
  outputs [#32, #33]
  properties {generalConfig {defaultMessage="If you'd like I could decode this little message of yours. "}}
 }
 TurnInQuest_C_4 {
  id=#33
  class=TurnInQuest_C
  outputs [#34]
  properties {softTaskIcon=&T_UI_NPC_Confirm}
  sub_nodes [ConversationChoiceNode_3 {class=ConversationChoiceNode, properties {defaultChoiceDisplayText=Accept}}]
 }
 GrantStaticQuest_C_0 {
  id=#34
  class=GrantStaticQuest_C
  outputs [#35]
  properties {questToGrantPtr=/BattlepassS16/Items/QuestItems/Narrative/quest_s16_tower_q07pre}
 }
 Speech_C_17 {
  id=#35
  class=Speech_C
  outputs [#36]
  properties {generalConfig {defaultMessage=It will take some time.}}
 }
 Speech_C_21 {
  id=#36
  class=Speech_C
  outputs [#37]
  properties {generalConfig {defaultMessage="And I will require a few favors in return. "}}
 }
 Speech_C_24 {
  id=#37
  class=Speech_C
  outputs []
  properties {generalConfig {defaultMessage="Now leave me alone. I have work to do. "}}
 }
 Back_C_0 {
  id=#32
  class=Back_C
  sub_nodes [ConversationChoiceNode_7 {class=ConversationChoiceNode, properties {defaultChoiceDisplayText=BACK}}]
 }
 Speech_C_45 {
  id=#7
  class=Speech_C
  outputs [#38]
  properties {softTaskIcon=&T_UI_NPC_MainQuest, generalConfig {defaultMessage="Do not rush the <b>Old Guardians. </>"}}
  sub_nodes [
   ConversationChoiceNode_0 {
    class=ConversationChoiceNode
    properties {defaultChoiceDisplayText=The Spire, choiceTags {gameplayTags [{tagName=Conversation.Choice.Category.MainQuestline}]}}
   }
   FortConversationRequirement_ControllerRequirement_1 {
    class=FortConversationRequirement_ControllerRequirement
    properties {
     nodeName=Require: Has Quest To Turn In
     requirement {
      class=FortControllerRequirement_HasQuestToTurnIn
      properties {softQuestToTurnIn=&quest_s16_tower_q08a}
     }
    }
   }
   FortConversationRequirement_ControllerRequirement_33 {
    class=FortConversationRequirement_ControllerRequirement
    properties {
     nodeName=Require: Has Quest To Turn In
     requirement {
      class=FortControllerRequirement_HasQuestToTurnIn
      properties {softQuestToTurnIn=&quest_s16_tower_q08b}
     }
    }
   }
   FortConversationRequirement_ControllerRequirement_34 {
    class=FortConversationRequirement_ControllerRequirement
    properties {
     nodeName=Require: Has Quest To Turn In
     requirement {
      class=FortControllerRequirement_HasQuestToTurnIn
      properties {softQuestToTurnIn=&quest_s16_tower_q08c}
     }
    }
   }
   FortConversationRequirement_ControllerRequirement_21 {
    class=FortConversationRequirement_ControllerRequirement
    properties {
     nodeName=Require: Has Completed Quest
     requirement {
      class=FortControllerRequirement_HasCompletedQuest
      properties {bInverseCondition=true, questDefinition=&quest_s16_tower_q08a}
     }
    }
   }
  ]
 }
 Speech_C_50 {
  id=#38
  class=Speech_C
  outputs [#39]
  properties {
   softTaskIcon=&T_UI_NPC_MainQuest
   generalConfig {defaultMessage="They will work their magic when they see fit. "}
  }
 }
 TurnInQuest_C_7 {id=#39, class=TurnInQuest_C, outputs [#40]}
 GrantStaticQuest_C_9 {
  id=#40
  class=GrantStaticQuest_C
  outputs []
  properties {questToGrantPtr=/BattlepassS16/Items/QuestItems/Narrative/quest_s16_tower_q09pre}
 }
 Speech_C_22 {
  id=#6
  class=Speech_C
  outputs [#41]
  properties {
   softTaskIcon=&T_UI_NPC_MainQuest
   generalConfig {defaultMessage="This is proving more <b>difficult</> than I had hoped. "}
  }
  sub_nodes [
   ConversationChoiceNode_1 {
    class=ConversationChoiceNode
    properties {defaultChoiceDisplayText=The Spire, choiceTags {gameplayTags [{tagName=Conversation.Choice.Category.MainQuestline}]}}
   }
   FortConversationRequirement_ControllerRequirement_7 {
    class=FortConversationRequirement_ControllerRequirement
    properties {
     nodeName=Require: Has Completed Quest
     requirement {
      class=FortControllerRequirement_HasCompletedQuest
      properties {questDefinition=/BattlepassS16/Items/QuestItems/Narrative/quest_s16_tower_q10}
     }
    }
   }
   FortConversationRequirement_ControllerRequirement_12 {
    class=FortConversationRequirement_ControllerRequirement
    properties {
     nodeName=Require: Has Completed Quest
     requirement {
      class=FortControllerRequirement_HasCompletedQuest
      properties {bInverseCondition=true, questDefinition=/BattlepassS16/Items/QuestItems/Narrative/quest_s16_tower_q11}
     }
    }
   }
  ]
 }
 Speech_C_23 {
  id=#41
  class=Speech_C
  outputs []
  properties {softTaskIcon=&T_UI_NPC_MainQuest, generalConfig {defaultMessage="I need more time. "}}
 }
 Speech_C_19 {
  id=#5
  class=Speech_C
  outputs [#42]
  properties {
   softTaskIcon=&T_UI_NPC_MainQuest
   generalConfig {defaultMessage="There is old magic in <enemy>the Spire</>. Be cautious in your explorations. "}
  }
  sub_nodes [
   ConversationChoiceNode_15 {
    class=ConversationChoiceNode
    properties {defaultChoiceDisplayText=The Spire, choiceTags {gameplayTags [{tagName=Conversation.Choice.Category.MainQuestline}]}}
   }
   FortConversationRequirement_ControllerRequirement_10 {
    class=FortConversationRequirement_ControllerRequirement
    properties {
     nodeName=Require: Is Quest Active
     requirement {class=FortControllerRequirement_IsQuestActive, properties {questDefinition=&quest_s16_tower_q10}}
    }
   }
   FortConversationRequirement_ControllerRequirement_26 {
    class=FortConversationRequirement_ControllerRequirement
    properties {
     nodeName=Require: Has Completed Quest
     requirement {
      class=FortControllerRequirement_HasCompletedQuest
      properties {bInverseCondition=true, questDefinition=&quest_s16_tower_q10}
     }
    }
   }
  ]
 }
 Speech_C_20 {
  id=#42
  class=Speech_C
  outputs [#43]
  properties {
   softTaskIcon=&T_UI_NPC_MainQuest
   generalConfig {defaultMessage="Return to me when you have gathered what you need from the<b> Crystal Shards. </>"}
  }
 }
 Speech_C_18 {
  id=#43
  class=Speech_C
  outputs []
  properties {
   softTaskIcon=&T_UI_NPC_MainQuest
   generalConfig {defaultMessage="May the <b>Old Guardians</> guide you home again."}
  }
 }
 Speech_C_48 {
  id=#4
  class=Speech_C
  outputs [#44]
  properties {
   softTaskIcon=&T_UI_NPC_MainQuest
   generalConfig {defaultMessage="The <b>Old Guardians</> smile upon our efforts. "}
  }
  sub_nodes [
   ConversationChoiceNode_24 {
    class=ConversationChoiceNode
    properties {defaultChoiceDisplayText=The Spire, choiceTags {gameplayTags [{tagName=Conversation.Choice.Category.MainQuestline}]}}
   }
   FortConversationRequirement_ControllerRequirement_35 {
    class=FortConversationRequirement_ControllerRequirement
    properties {
     nodeName=Require: Has Quest To Turn In
     requirement {
      class=FortControllerRequirement_HasQuestToTurnIn
      properties {softQuestToTurnIn=&quest_s16_tower_q09}
     }
    }
   }
   FortConversationRequirement_ControllerRequirement_25 {
    class=FortConversationRequirement_ControllerRequirement
    properties {
     nodeName=Require: Has Completed Quest
     requirement {
      class=FortControllerRequirement_HasCompletedQuest
      properties {bInverseCondition=true, questDefinition=&quest_s16_tower_q09}
     }
    }
   }
   FortConversationRequirement_ControllerRequirement_40 {
    class=FortConversationRequirement_ControllerRequirement
    properties {
     nodeName=Require: Has Completed Quest
     requirement {
      class=FortControllerRequirement_HasCompletedQuest
      properties {bInverseCondition=true, questDefinition=&quest_s16_tower_q10}
     }
    }
   }
   FortConversationRequirement_ControllerRequirement_41 {
    class=FortConversationRequirement_ControllerRequirement
    properties {
     nodeName=Require: Is Quest Active
     requirement {
      class=FortControllerRequirement_IsQuestActive
      properties {bInverseCondition=true, questDefinition=&quest_s16_tower_q10}
     }
    }
   }
  ]
 }
 Speech_C_49 {
  id=#44
  class=Speech_C
  outputs [#45]
  properties {
   softTaskIcon=&T_UI_NPC_MainQuest
   generalConfig {defaultMessage="Your disguise is waiting for you near <enemy>the Spire. </>"}
  }
 }
 TurnInQuest_C_8 {id=#45, class=TurnInQuest_C, outputs [#46]}
 GrantStaticQuest_C_10 {
  id=#46
  class=GrantStaticQuest_C
  outputs [#47]
  properties {questToGrantPtr=&quest_s16_tower_q10}
 }
 Speech_C_51 {
  id=#47
  class=Speech_C
  outputs [#48]
  properties {
   softTaskIcon=&T_UI_NPC_MainQuest
   generalConfig {defaultMessage="Among your things you will find a <b>Crystal Focus</> pickaxe. "}
  }
 }
 Speech_C_7 {
  id=#48
  class=Speech_C
  outputs [#49]
  properties {
   softTaskIcon=&T_UI_NPC_MainQuest
   generalConfig {defaultMessage="Sample the Crystal Shards and <b>return the device</> to me."}
  }
 }
 Speech_C_8 {
  id=#49
  class=Speech_C
  outputs []
  properties {softTaskIcon=&T_UI_NPC_MainQuest, generalConfig {defaultMessage="May the <b>Old Guardians</> guide your path. "}}
 }
 Speech_C_0 {
  id=#3
  class=Speech_C
  outputs [#50]
  properties {
   softTaskIcon=&T_UI_NPC_MainQuest
   generalConfig {defaultMessage="Thief? Occasionally. But this time I had something else in mind. "}
  }
  sub_nodes [
   ConversationChoiceNode_6 {
    class=ConversationChoiceNode
    properties {defaultChoiceDisplayText=The Spire, choiceTags {gameplayTags [{tagName=Conversation.Choice.Category.MainQuestline}]}}
   }
   FortConversationRequirement_ControllerRequirement_2 {
    class=FortConversationRequirement_ControllerRequirement
    properties {
     nodeName=Require: Has Completed Quest
     requirement {
      class=FortControllerRequirement_HasCompletedQuest
      properties {bInverseCondition=true, questDefinition=/BattlepassS16/Items/QuestItems/Narrative/quest_s16_tower_q05}
     }
    }
   }
   FortConversationRequirement_ControllerRequirement_13 {
    class=FortConversationRequirement_ControllerRequirement
    properties {
     nodeName=Require: Has Quest To Turn In
     requirement {
      class=FortControllerRequirement_HasQuestToTurnIn
      properties {softQuestToTurnIn=&quest_s16_tower_q05}
     }
    }
   }
  ]
 }
 TurnInQuest_C_0 {id=#50, class=TurnInQuest_C, outputs [#51]}
 GrantStaticQuest_C_1 {
  id=#51
  class=GrantStaticQuest_C
  outputs [#52]
  properties {questToGrantPtr=&quest_s16_tower_q06}
 }
 Speech_C_4 {
  id=#52
  class=Speech_C
  outputs [#53]
  properties {generalConfig {defaultMessage="That <b>device</> of yours is waiting for you nearby. "}}
  sub_nodes [ConversationChoiceNode_19 {class=ConversationChoiceNode, properties {defaultChoiceDisplayText=Continue}}]
 }
 Speech_C_1 {
  id=#53
  class=Speech_C
  outputs []
  properties {generalConfig {defaultMessage="It's very interesting... <b>once you understand it. </>"}}
  sub_nodes [ConversationChoiceNode_20 {class=ConversationChoiceNode, properties {defaultChoiceDisplayText=Continue}}]
 }
 Speech_C_2 {
  id=#2
  class=Speech_C
  outputs []
  properties {softTaskIcon=&T_UI_NPC_MainQuest, generalConfig {defaultMessage=Listen first. Talk later.}}
  sub_nodes [
   ConversationChoiceNode_10 {
    class=ConversationChoiceNode
    properties {defaultChoiceDisplayText=The Spire, choiceTags {gameplayTags [{tagName=Conversation.Choice.Category.MainQuestline}]}}
   }
   FortConversationRequirement_ControllerRequirement_22 {
    class=FortConversationRequirement_ControllerRequirement
    properties {
     nodeName=Require: Is Quest Active
     requirement {class=FortControllerRequirement_IsQuestActive, properties {questDefinition=&quest_s16_tower_q06}}
    }
   }
   FortConversationRequirement_ControllerRequirement_23 {
    class=FortConversationRequirement_ControllerRequirement
    properties {
     nodeName=Require: Has Quest To Turn In
     requirement {
      class=FortControllerRequirement_HasQuestToTurnIn
      properties {bInverseCondition=true, softQuestToTurnIn=&quest_s16_tower_q06}
     }
    }
   }
  ]
 }
]
