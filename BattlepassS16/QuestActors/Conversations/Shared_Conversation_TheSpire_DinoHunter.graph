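# Conversation database for the DinoHunter NPC in the Season 16 Spire questline, covering the handoff from tower quest q03 through q05.
# One entry point offers three "The Spire" main-quest choices, gated by quest-state requirements.
# The offer branch has Back or Accept, which turns in a quest and grants q04's artifact hunt.
# Other branches nag while q04 is unfinished, or reveal Raz's theft, turn in and grant q05.
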ConversationDatabase /BattlepassS16/QuestActors/Conversations/Shared_Conversation_TheSpire_DinoHunter
entry_points [{tag=Conversation.Entry.Season16.TheSpire.DinoHunter, destinations [#1]}]
nodes [
 ConversationEntryPointNode_7 {
  id=#1
  class=ConversationEntryPointNode
  outputs [#2, #3, #4]
  properties {entryTag {tagName=Conversation.Entry.Season16.TheSpire.DinoHunter}}
 }
 Speech_C_2 {
  id=#4
  class=Speech_C
  outputs []
  properties {softTaskIcon=&T_UI_NPC_MainQuest, generalConfig {defaultMessage="Nope. Not until you have my artifacts. "}}
  sub_nodes [
   ConversationChoiceNode_4 {
    class=ConversationChoiceNode
    properties {defaultChoiceDisplayText=The Spire, choiceTags {gameplayTags [{tagName=Conversation.Choice.Category.MainQuestline}]}}
   }
   FortConversationRequirement_ControllerRequirement_1 {
    class=FortConversationRequirement_ControllerRequirement
    properties {
     nodeName=Require: Is Quest Active
     requirement {class=FortControllerRequirement_IsQuestActive, properties {questDefinition=/BattlepassS16/Items/QuestItems/Narrative/quest_s16_tower_q04}}
    }
   }
   FortConversationRequirement_ControllerRequirement_12 {
    class=FortConversationRequirement_ControllerRequirement
    properties {
     nodeName=Require: Has Quest To Turn In
     requirement {
      class=FortControllerRequirement_HasQuestToTurnIn
      properties {bInverseCondition=true, softQuestToTurnIn=&quest_s16_tower_q04}
     }
    }
   }
  ]
 }
 Speech_C_0 {
  id=#3
  class=Speech_C
  outputs [#5]
  properties {softTaskIcon=&T_UI_NPC_MainQuest, generalConfig {defaultMessage="Hmm. Very nice. You did well. "}}
  sub_nodes [
   ConversationChoiceNode_5 {
    class=ConversationChoiceNode
    properties {defaultChoiceDisplayText=The Spire, choiceTags {gameplayTags [{tagName=Conversation.Choice.Category.MainQuestline}]}}
   }
   FortConversationRequirement_ControllerRequirement_2 {
    class=FortConversationRequirement_ControllerRequirement
    properties {
     nodeName=Require: Has Quest To Turn In
     requirement {
      class=FortControllerRequirement_HasQuestToTurnIn
      properties {softQuestToTurnIn=&quest_s16_tower_q04}
     }
    }
   }
   FortConversationRequirement_ControllerRequirement_4 {
    class=FortConversationRequirement_ControllerRequirement
    properties {
     nodeName=Require: Has Completed Quest
     requirement {
      class=FortControllerRequirement_HasCompletedQuest
      properties {bInverseCondition=true, questDefinition=&quest_s16_tower_q04}
     }
    }
   }
  ]
 }
 Speech_C_4 {
  id=#5
  class=Speech_C
  outputs [#6]
  properties {generalConfig {defaultMessage="I cannot keep my promise. Your device was <b>stolen</> by that thief, <enemy>Raz.  </>"}}
 }
 Speech_C_8 {
  id=#6
  class=Speech_C
  outputs [#7]
  properties {generalConfig {defaultMessage="He's easy enough to find, though. "}}
 }
 Speech_C_9 {
  id=#7
  class=Speech_C
  outputs [#8]
  properties {generalConfig {defaultMessage="You'll likely find him in <b>Colossal Crops</> hawking the treasure's he's \"found.\" "}}
 }
 TurnInQuest_C_0 {id=#8, class=TurnInQuest_C, outputs [#9]}
 GrantStaticQuest_C_1 {
  id=#9
  class=GrantStaticQuest_C
  outputs []
  properties {questToGrantPtr=/BattlepassS16/Items/QuestItems/Narrative/quest_s16_tower_q05}
 }
 Speech_C_1 {
  id=#2
  class=Speech_C
  outputs [#10]
  properties {
   softTaskIcon=&T_UI_NPC_MainQuest
   generalConfig {defaultMessage="I found something for you. A strange device with some sort of message... I think. "}
  }
  sub_nodes [
   ConversationChoiceNode_6 {
    class=ConversationChoiceNode
    properties {defaultChoiceDisplayText=The Spire, choiceTags {gameplayTags [{tagName=Conversation.Choice.Category.MainQuestline}]}}
   }
   FortConversationRequirement_ControllerRequirement_6 {
    class=FortConversationRequirement_ControllerRequirement
    properties {
     nodeName=Require: Has Completed Quest
     requirement {
      class=FortControllerRequirement_HasCompletedQuest
      properties {bInverseCondition=true, questDefinition=&quest_s16_tower_q04}
     }
    }
   }
   FortConversationRequirement_ControllerRequirement_7 {
    class=FortConversationRequirement_ControllerRequirement
    properties {
     nodeName=Require: Is Quest Active
     requirement {
      class=FortControllerRequirement_IsQuestActive
      properties {bInverseCondition=true, questDefinition=&quest_s16_tower_q04}
     }
    }
   }
   FortConversationRequirement_ControllerRequirement_0 {
    class=FortConversationRequirement_ControllerRequirement
    properties {
     nodeName=Require: Has Quest To Turn In
     requirement {
      class=FortControllerRequirement_HasQuestToTurnIn
      properties {softQuestToTurnIn=/BattlepassS16/Items/QuestItems/Narrative/quest_s16_tower_q03}
     }
    }
   }
   FortConversationRequirement_ControllerRequirement_3 {
    class=FortConversationRequirement_ControllerRequirement
    properties {
     nodeName=Require: Has Completed Quest
     requirement {
      class=FortControllerRequirement_HasCompletedQuest
      properties {bInverseCondition=true, questDefinition=&quest_s16_tower_q03}
     }
    }
   }
   FortConversationRequirement_ControllerRequirement_5 {
    class=FortConversationRequirement_ControllerRequirement
    properties {
     nodeName=Require: Has Completed Quest
     requirement {
      class=FortControllerRequirement_HasCompletedQuest
      properties {questDefinition=/BattlepassS16/Items/QuestItems/Narrative/quest_s16_tower_q02}
     }
    }
   }
  ]
 }
 Speech_C_7 {
  id=#10
  class=Speech_C
  outputs [#11, #12]
  properties {generalConfig {defaultMessage="Collect some <b>artifacts</> for me and it's yours. "}}
  sub_nodes [ConversationChoiceNode_11 {class=ConversationChoiceNode, properties {defaultChoiceDisplayText=Continue}}]
 }
 TurnInQuest_C_1 {
  id=#12
  class=TurnInQuest_C
  outputs [#13]
  properties {softTaskIcon=/Game/Athena/HUD/Interaction/T_UI_NPC_Confirm}
  sub_nodes [ConversationChoiceNode_8 {class=ConversationChoiceNode, properties {defaultChoiceDisplayText=ACCEPT}}]
 }
 GrantStaticQuest_C_0 {
  id=#13
  class=GrantStaticQuest_C
  outputs [#14]
  properties {questToGrantPtr=&quest_s16_tower_q04}
 }
 Speech_C_6 {
  id=#14
  class=Speech_C
  outputs []
  properties {generalConfig {defaultMessage="Stay close, there are plenty of good artifacts nearby. "}}
  sub_nodes [ConversationChoiceNode_3 {class=ConversationChoiceNode, properties {defaultChoiceDisplayText=Continue}}]
 }
 Back_C_0 {
  id=#11
  class=Back_C
  sub_nodes [ConversationChoiceNode_10 {class=ConversationChoiceNode, properties {defaultChoiceDisplayText=BACK}}]
 }
]
