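# Shared conversation database for the Season 16 Spire questline, offering "The Spire" dialogue with the various Jones NPCs.
# One entry point fans out to speech chains, each gated by quest state and the NPC's entry tag, so each Jones shows only his own lines.
# Most chains end by completing a tower_q14 objective, while Jonesy the First's chains handle q14 turn-in, granting q16 and q16 turn-in.
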ConversationDatabase /BattlepassS16/QuestActors/Conversations/Shared_Conversation_TheSpire_Joneses
entry_points [{tag=Conversation.Entry.Season16.TheSpire.Joneses, destinations [#1]}]
nodes [
 ConversationEntryPointNode_7 {
  id=#1
  class=ConversationEntryPointNode
  outputs [
   #2
   #3
   #4
   #5
   #6
   #7
   #8
   #9
   #10
   #11
   #12
   #13
   #14
  ]
  properties {entryTag {tagName=Conversation.Entry.Season16.TheSpire.Joneses}}
 }
 Speech_C_25 {
  id=#14
  class=Speech_C
  outputs [#15]
  properties {
   softTaskIcon=&T_UI_NPC_MainQuest
   generalConfig {defaultMessage="<b>Not bad. </>\r\nGuess I underestimated you. "}
  }
  sub_nodes [
   ConversationChoiceNode_19 {
    class=ConversationChoiceNode
    properties {defaultChoiceDisplayText=The Spire, choiceTags {gameplayTags [{tagName=Conversation.Choice.Category.MainQuestline}]}}
   }
   FortConversationRequirement_ControllerRequirement_32 {
    class=FortConversationRequirement_ControllerRequirement
    properties {
     nodeName=Require: Has Completed Quest
     requirement {
      class=FortControllerRequirement_HasCompletedQuest
      properties {bInverseCondition=true, questDefinition=/BattlepassS16/Items/QuestItems/Narrative/quest_s16_tower_q16}
     }
    }
   }
   FortConversationRequirement_ControllerRequirement_33 {
    class=FortConversationRequirement_ControllerRequirement
    properties {
     nodeName=Require: Is Quest Active
     requirement {class=FortControllerRequirement_IsQuestActive, properties {questDefinition=&quest_s16_tower_q16}}
    }
   }
   FortConversationRequirement_ControllerRequirement_47 {
    class=FortConversationRequirement_ControllerRequirement
    properties {
     nodeName=Require: Has Quest To Turn In
     requirement {
      class=FortControllerRequirement_HasQuestToTurnIn
      properties {softQuestToTurnIn=&quest_s16_tower_q16}
     }
    }
   }
   FortConversationRequirement_ControllerRequirement_50 {
    class=FortConversationRequirement_ControllerRequirement
    properties {
     nodeName=Require: NPC Has Conversation Tag
     participantId {tagName=Conversation.Participant.NPC}
     requirement {class=FortControllerRequirement_NPCHasConversationEntryTag, properties {queryTag {tagName=Conversation.Entry.JonesyTheFirst}}}
    }
   }
  ]
 }
 Speech_C_26 {
  id=#15
  class=Speech_C
  outputs [#16]
  properties {generalConfig {defaultMessage="There are people who think <b>the Spire</> will make them <b>powerful.</>"}}
  sub_nodes [ConversationChoiceNode_29 {class=ConversationChoiceNode, properties {defaultChoiceDisplayText=Continue}}]
 }
 Speech_C_28 {
  id=#16
  class=Speech_C
  outputs [#17]
  properties {generalConfig {defaultMessage="It will, but it will also exploit their every <b>weakness. </>"}}
  sub_nodes [ConversationChoiceNode_31 {class=ConversationChoiceNode, properties {defaultChoiceDisplayText=Continue}}]
 }
 Speech_C_27 {
  id=#17
  class=Speech_C
  outputs [#18]
  properties {generalConfig {defaultMessage="Do not allow <b>the Spire</> to corrupt someone <enemy>powerful</>"}}
  sub_nodes [ConversationChoiceNode_32 {class=ConversationChoiceNode, properties {defaultChoiceDisplayText=Continue}}]
 }
 Speech_C_29 {
  id=#18
  class=Speech_C
  outputs [#19]
  properties {generalConfig {defaultMessage="or <b>Jones</> does not stand a chance."}}
  sub_nodes [ConversationChoiceNode_53 {class=ConversationChoiceNode, properties {defaultChoiceDisplayText=Continue}}]
 }
 Speech_C_51 {
  id=#19
  class=Speech_C
  outputs [#20]
  properties {generalConfig {defaultMessage="Go now, before <b>the Spire</> attracts any more attention from the - "}}
  sub_nodes [ConversationChoiceNode_54 {class=ConversationChoiceNode, properties {defaultChoiceDisplayText=Continue}}]
 }
 Speech_C_52 {
  id=#20
  class=Speech_C
  outputs [#21]
  properties {generalConfig {defaultMessage="Never mind, just <b>hurry!</>"}}
  sub_nodes [ConversationChoiceNode_30 {class=ConversationChoiceNode, properties {defaultChoiceDisplayText=Continue}}]
 }
 TurnInQuest_C_1 {id=#21, class=TurnInQuest_C, outputs []}
 Speech_C_24 {
  id=#13
  class=Speech_C
  outputs [#22]
  properties {softTaskIcon=&T_UI_NPC_MainQuest, generalConfig {defaultMessage=You want my help?}}
  sub_nodes [
   ConversationChoiceNode_27 {
    class=ConversationChoiceNode
    properties {defaultChoiceDisplayText=The Spire, choiceTags {gameplayTags [{tagName=Conversation.Choice.Category.MainQuestline}]}}
   }
   FortConversationRequirement_ControllerRequirement_45 {
    class=FortConversationRequirement_ControllerRequirement
    properties {
     nodeName=Require: Has Completed Quest
     requirement {
      class=FortControllerRequirement_HasCompletedQuest
      properties {bInverseCondition=true, questDefinition=&quest_s16_tower_q16}
     }
    }
   }
   FortConversationRequirement_ControllerRequirement_46 {
    class=FortConversationRequirement_ControllerRequirement
    properties {
     nodeName=Require: Is Quest Active
     requirement {class=FortControllerRequirement_IsQuestActive, properties {questDefinition=&quest_s16_tower_q16}}
    }
   }
   FortConversationRequirement_ControllerRequirement_49 {
    class=FortConversationRequirement_ControllerRequirement
    properties {
     nodeName=Require: NPC Has Conversation Tag
     participantId {tagName=Conversation.Participant.NPC}
     requirement {class=FortControllerRequirement_NPCHasConversationEntryTag, properties {queryTag {tagName=Conversation.Entry.JonesyTheFirst}}}
    }
   }
   FortConversationRequirement_ControllerRequirement_34 {
    class=FortConversationRequirement_ControllerRequirement
    properties {
     nodeName=Require: Has Quest To Turn In
     requirement {
      class=FortControllerRequirement_HasQuestToTurnIn
      properties {bInverseCondition=true, softQuestToTurnIn=&quest_s16_tower_q16}
     }
    }
   }
  ]
 }
 Speech_C_47 {
  id=#22
  class=Speech_C
  outputs []
  properties {generalConfig {defaultMessage="Get ready to <enemy>fight!</>"}}
  sub_nodes [ConversationChoiceNode_28 {class=ConversationChoiceNode, properties {defaultChoiceDisplayText=Continue}}]
 }
 Speech_C_17 {
  id=#12
  class=Speech_C
  outputs [#23]
  properties {
   softTaskIcon=&T_UI_NPC_MainQuest
   generalConfig {defaultMessage="It looks like Jones asked the <b>right person</> for help. "}
  }
  sub_nodes [
   ConversationChoiceNode_26 {
    class=ConversationChoiceNode
    properties {defaultChoiceDisplayText=The Spire, choiceTags {gameplayTags [{tagName=Conversation.Choice.Category.MainQuestline}]}}
   }
   FortConversationRequirement_ControllerRequirement_39 {
    class=FortConversationRequirement_ControllerRequirement
    properties {
     nodeName=Require: Has Completed Quest
     requirement {
      class=FortControllerRequirement_HasCompletedQuest
      properties {bInverseCondition=true, questDefinition=/BattlepassS16/Items/QuestItems/Narrative/quest_s16_tower_q14}
     }
    }
   }
   FortConversationRequirement_ControllerRequirement_43 {
    class=FortConversationRequirement_ControllerRequirement
    properties {
     nodeName=Require: Is Quest Active
     requirement {class=FortControllerRequirement_IsQuestActive, properties {questDefinition=&quest_s16_tower_q14}}
    }
   }
   FortConversationRequirement_ControllerRequirement_44 {
    class=FortConversationRequirement_ControllerRequirement
    properties {
     nodeName=Require: Has Quest To Turn In
     requirement {
      class=FortControllerRequirement_HasQuestToTurnIn
      properties {softQuestToTurnIn=&quest_s16_tower_q14}
     }
    }
   }
   FortConversationRequirement_ControllerRequirement_48 {
    class=FortConversationRequirement_ControllerRequirement
    properties {
     nodeName=Require: NPC Has Conversation Tag
     participantId {tagName=Conversation.Participant.NPC}
     requirement {class=FortControllerRequirement_NPCHasConversationEntryTag, properties {queryTag {tagName=Conversation.Entry.JonesyTheFirst}}}
    }
   }
   FortConversationRequirement_ControllerRequirement_51 {
    class=FortConversationRequirement_ControllerRequirement
    properties {
     nodeName=Require: Is Quest Active
     requirement {
      class=FortControllerRequirement_IsQuestActive
      properties {bInverseCondition=true, questDefinition=&quest_s16_tower_q16}
     }
    }
   }
  ]
 }
 Speech_C_18 {
  id=#23
  class=Speech_C
  outputs [#24]
  properties {generalConfig {defaultMessage="If you <b>really</> want to get wrapped up in this <b>maybe</> I'll help. "}}
  sub_nodes [ConversationChoiceNode_49 {class=ConversationChoiceNode, properties {defaultChoiceDisplayText=Continue}}]
 }
 Speech_C_22 {
  id=#24
  class=Speech_C
  outputs [#25]
  properties {generalConfig {defaultMessage=Or...}}
  sub_nodes [ConversationChoiceNode_25 {class=ConversationChoiceNode, properties {defaultChoiceDisplayText=Continue}}]
 }
 Speech_C_23 {
  id=#25
  class=Speech_C
  outputs [#26]
  properties {generalConfig {defaultMessage="Maybe we'll end this right here. "}}
  sub_nodes [ConversationChoiceNode_48 {class=ConversationChoiceNode, properties {defaultChoiceDisplayText=Continue}}]
 }
 Speech_C_46 {
  id=#26
  class=Speech_C
  outputs [#27]
  properties {generalConfig {defaultMessage="Up to you. "}}
  sub_nodes [ConversationChoiceNode_11 {class=ConversationChoiceNode, properties {defaultChoiceDisplayText=Continue}}]
 }
 TurnInQuest_C_0 {id=#27, class=TurnInQuest_C, outputs [#28]}
 GrantStaticQuest_C_0 {
  id=#28
  class=GrantStaticQuest_C
  outputs []
  properties {questToGrantPtr=&quest_s16_tower_q16}
 }
 Speech_C_20 {
  id=#11
  class=Speech_C
  outputs [#29]
  properties {
   softTaskIcon=&T_UI_NPC_MainQuest
   generalConfig {defaultMessage="I was involved in <b>The Spire? </>\r\nLooks like I'm never earning my <b>Cautious Camper Sash Badge.</> "}
  }
  sub_nodes [
   ConversationChoiceNode_15 {
    class=ConversationChoiceNode
    properties {defaultChoiceDisplayText=The Spire, choiceTags {gameplayTags [{tagName=Conversation.Choice.Category.MainQuestline}]}}
   }
   FortConversationRequirement_ControllerRequirement_24 {
    class=FortConversationRequirement_ControllerRequirement
    properties {
     nodeName=Require: Has Completed Quest
     requirement {
      class=FortControllerRequirement_HasCompletedQuest
      properties {bInverseCondition=true, questDefinition=&quest_s16_tower_q14}
     }
    }
   }
   FortConversationRequirement_ControllerRequirement_25 {
    class=FortConversationRequirement_ControllerRequirement
    properties {
     nodeName=Require: Is Quest Active
     requirement {class=FortControllerRequirement_IsQuestActive, properties {questDefinition=&quest_s16_tower_q14}}
    }
   }
   FortConversationRequirement_ControllerRequirement_26 {
    class=FortConversationRequirement_ControllerRequirement
    properties {
     nodeName=Require: Has Completed Quest Objective
     requirement {
      class=FortControllerRequirement_HasCompletedQuestObjective
      properties {
       bInverseCondition=true
       objectiveBackendName=quest_s16_tower_q14_obj5
       questDefinition=&quest_s16_tower_q14
      }
     }
    }
   }
   FortConversationRequirement_ControllerRequirement_27 {
    class=FortConversationRequirement_ControllerRequirement
    properties {
     nodeName=Require: NPC Has Conversation Tag
     participantId {tagName=Conversation.Participant.NPC}
     requirement {class=FortControllerRequirement_NPCHasConversationEntryTag, properties {queryTag {tagName=Conversation.Entry.SashSeargent}}}
    }
   }
  ]
 }
 Speech_C_21 {
  id=#29
  class=Speech_C
  outputs [#30]
  properties {generalConfig {defaultMessage="According to the <b>IO Agent Field Guide,</> memory loss is a very common when entering <b>the loop. </>"}}
  sub_nodes [ConversationChoiceNode_40 {class=ConversationChoiceNode, properties {defaultChoiceDisplayText=Continue}}]
 }
 Speech_C_38 {
  id=#30
  class=Speech_C
  outputs [#31]
  properties {generalConfig {defaultMessage="So I don't know if any of us can help you...\r\nunless..."}}
  sub_nodes [ConversationChoiceNode_41 {class=ConversationChoiceNode, properties {defaultChoiceDisplayText=Continue}}]
 }
 Speech_C_39 {
  id=#31
  class=Speech_C
  outputs [#32]
  properties {generalConfig {defaultMessage="... the rumors of <b>The First Jones</> are true."}}
  sub_nodes [ConversationChoiceNode_16 {class=ConversationChoiceNode, properties {defaultChoiceDisplayText=Continue}}]
 }
 Speech_C_40 {
  id=#32
  class=Speech_C
  outputs [#33]
  properties {generalConfig {defaultMessage="That's all you're getting from me. \r\nI'm not losing my <b>No Snitching Badge. </>"}}
  sub_nodes [ConversationChoiceNode_42 {class=ConversationChoiceNode, properties {defaultChoiceDisplayText=Continue}}]
 }
 FortConversationTaskNode_CompleteQuestObjectiveByBackendName_5 {
  id=#33
  class=FortConversationTaskNode_CompleteQuestObjectiveByBackendName
  properties {objectiveBackendName=quest_s16_tower_q14_obj5, questDefinition=&quest_s16_tower_q14}
  sub_nodes [FortConversationRequirement_ControllerRequirement_57 {
    class=FortConversationRequirement_ControllerRequirement
    properties {
     nodeName=Require: Has Quest To Turn In
     requirement {
      class=FortControllerRequirement_HasQuestToTurnIn
      properties {bInverseCondition=true, softQuestToTurnIn=&quest_s16_tower_q14}
     }
    }
   }]
 }
 Speech_C_19 {
  id=#10
  class=Speech_C
  outputs []
  properties {
   softTaskIcon=&T_UI_NPC_MainQuest
   generalConfig {defaultMessage="Prove to me you really want to learn what's going on here. Then <b>maybe</> we'll talk. "}
  }
  sub_nodes [
   ConversationChoiceNode_21 {
    class=ConversationChoiceNode
    properties {defaultChoiceDisplayText=The Spire, choiceTags {gameplayTags [{tagName=Conversation.Choice.Category.MainQuestline}]}}
   }
   FortConversationRequirement_ControllerRequirement_36 {
    class=FortConversationRequirement_ControllerRequirement
    properties {
     nodeName=Require: Has Completed Quest
     requirement {
      class=FortControllerRequirement_HasCompletedQuest
      properties {bInverseCondition=true, questDefinition=&quest_s16_tower_q14}
     }
    }
   }
   FortConversationRequirement_ControllerRequirement_37 {
    class=FortConversationRequirement_ControllerRequirement
    properties {
     nodeName=Require: Is Quest Active
     requirement {class=FortControllerRequirement_IsQuestActive, properties {questDefinition=&quest_s16_tower_q14}}
    }
   }
   FortConversationRequirement_ControllerRequirement_38 {
    class=FortConversationRequirement_ControllerRequirement
    properties {
     nodeName=Require: NPC Has Conversation Tag
     participantId {tagName=Conversation.Participant.NPC}
     requirement {class=FortControllerRequirement_NPCHasConversationEntryTag, properties {queryTag {tagName=Conversation.Entry.JonesyTheFirst}}}
    }
   }
   FortConversationRequirement_ControllerRequirement_35 {
    class=FortConversationRequirement_ControllerRequirement
    properties {
     nodeName=Require: Has Quest To Turn In
     requirement {
      class=FortControllerRequirement_HasQuestToTurnIn
      properties {bInverseCondition=true, softQuestToTurnIn=&quest_s16_tower_q14}
     }
    }
   }
  ]
 }
 Speech_C_15 {
  id=#9
  class=Speech_C
  outputs [#34]
  properties {
   softTaskIcon=&T_UI_NPC_MainQuest
   generalConfig {defaultMessage="Sometimes bits and pieces come back to me when I'm <b>on a dive. </>"}
  }
  sub_nodes [
   ConversationChoiceNode_8 {
    class=ConversationChoiceNode
    properties {defaultChoiceDisplayText=The Spire, choiceTags {gameplayTags [{tagName=Conversation.Choice.Category.MainQuestline}]}}
   }
   FortConversationRequirement_ControllerRequirement_1 {
    class=FortConversationRequirement_ControllerRequirement
    properties {
     nodeName=Require: Has Completed Quest
     requirement {
      class=FortControllerRequirement_HasCompletedQuest
      properties {bInverseCondition=true, questDefinition=&quest_s16_tower_q14}
     }
    }
   }
   FortConversationRequirement_ControllerRequirement_3 {
    class=FortConversationRequirement_ControllerRequirement
    properties {
     nodeName=Require: Is Quest Active
     requirement {class=FortControllerRequirement_IsQuestActive, properties {questDefinition=&quest_s16_tower_q14}}
    }
   }
   FortConversationRequirement_ControllerRequirement_7 {
    class=FortConversationRequirement_ControllerRequirement
    properties {
     nodeName=Require: Has Completed Quest Objective
     requirement {
      class=FortControllerRequirement_HasCompletedQuestObjective
      properties {
       bInverseCondition=true
       objectiveBackendName=quest_s16_tower_q14_obj8
       questDefinition=&quest_s16_tower_q14
      }
     }
    }
   }
   FortConversationRequirement_ControllerRequirement_11 {
    class=FortConversationRequirement_ControllerRequirement
    properties {
     nodeName=Require: NPC Has Conversation Tag
     participantId {tagName=Conversation.Participant.NPC}
     requirement {class=FortControllerRequirement_NPCHasConversationEntryTag, properties {queryTag {tagName=Conversation.Entry.WreckRaider}}}
    }
   }
  ]
 }
 Speech_C_16 {
  id=#34
  class=Speech_C
  outputs [#35]
  properties {generalConfig {defaultMessage="I came here on a <b>mission to study the zero point...</>"}}
  sub_nodes [ConversationChoiceNode_46 {class=ConversationChoiceNode, properties {defaultChoiceDisplayText=Continue}}]
 }
 Speech_C_44 {
  id=#35
  class=Speech_C
  outputs [#36]
  properties {generalConfig {defaultMessage="But once you're in the loop you're just an <b>echo of who you once were...</>"}}
  sub_nodes [ConversationChoiceNode_22 {class=ConversationChoiceNode, properties {defaultChoiceDisplayText=Continue}}]
 }
 Speech_C_45 {
  id=#36
  class=Speech_C
  outputs [#37]
  properties {generalConfig {defaultMessage="Well... that's true for <b>most</> of us...."}}
  sub_nodes [ConversationChoiceNode_47 {class=ConversationChoiceNode, properties {defaultChoiceDisplayText=Continue}}]
 }
 FortConversationTaskNode_CompleteQuestObjectiveByBackendName_8 {
  id=#37
  class=FortConversationTaskNode_CompleteQuestObjectiveByBackendName
  properties {objectiveBackendName=quest_s16_tower_q14_obj8, questDefinition=&quest_s16_tower_q14}
  sub_nodes [FortConversationRequirement_ControllerRequirement_65 {
    class=FortConversationRequirement_ControllerRequirement
    properties {
     nodeName=Require: Has Quest To Turn In
     requirement {
      class=FortControllerRequirement_HasQuestToTurnIn
      properties {bInverseCondition=true, softQuestToTurnIn=&quest_s16_tower_q14}
     }
    }
   }]
 }
 Speech_C_13 {
  id=#8
  class=Speech_C
  outputs [#38]
  properties {
   softTaskIcon=&T_UI_NPC_MainQuest
   generalConfig {defaultMessage="Right on, man. \r\nI always thought I'd be <b>a hero. </>"}
  }
  sub_nodes [
   ConversationChoiceNode_17 {
    class=ConversationChoiceNode
    properties {defaultChoiceDisplayText=The Spire, choiceTags {gameplayTags [{tagName=Conversation.Choice.Category.MainQuestline}]}}
   }
   FortConversationRequirement_ControllerRequirement_28 {
    class=FortConversationRequirement_ControllerRequirement
    properties {
     nodeName=Require: Has Completed Quest
     requirement {
      class=FortControllerRequirement_HasCompletedQuest
      properties {bInverseCondition=true, questDefinition=&quest_s16_tower_q14}
     }
    }
   }
   FortConversationRequirement_ControllerRequirement_29 {
    class=FortConversationRequirement_ControllerRequirement
    properties {
     nodeName=Require: Is Quest Active
     requirement {class=FortControllerRequirement_IsQuestActive, properties {questDefinition=&quest_s16_tower_q14}}
    }
   }
   FortConversationRequirement_ControllerRequirement_30 {
    class=FortConversationRequirement_ControllerRequirement
    properties {
     nodeName=Require: Has Completed Quest Objective
     requirement {
      class=FortControllerRequirement_HasCompletedQuestObjective
      properties {
       bInverseCondition=true
       objectiveBackendName=quest_s16_tower_q14_obj7
       questDefinition=&quest_s16_tower_q14
      }
     }
    }
   }
   FortConversationRequirement_ControllerRequirement_31 {
    class=FortConversationRequirement_ControllerRequirement
    properties {
     nodeName=Require: NPC Has Conversation Tag
     participantId {tagName=Conversation.Participant.NPC}
     requirement {class=FortControllerRequirement_NPCHasConversationEntryTag, properties {queryTag {tagName=Conversation.Entry.SuntanSpecialist}}}
    }
   }
  ]
 }
 Speech_C_14 {
  id=#38
  class=Speech_C
  outputs [#39]
  properties {generalConfig {defaultMessage="But if I know me, I probably don't want to be stuck in a <b>weird limbo between worlds. </>"}}
  sub_nodes [ConversationChoiceNode_45 {class=ConversationChoiceNode, properties {defaultChoiceDisplayText=Continue}}]
 }
 Speech_C_43 {
  id=#39
  class=Speech_C
  outputs [#40]
  properties {generalConfig {defaultMessage="Maybe one of the <b>other Joneses</> can help. "}}
  sub_nodes [ConversationChoiceNode_18 {class=ConversationChoiceNode, properties {defaultChoiceDisplayText=Continue}}]
 }
 FortConversationTaskNode_CompleteQuestObjectiveByBackendName_7 {
  id=#40
  class=FortConversationTaskNode_CompleteQuestObjectiveByBackendName
  properties {objectiveBackendName=quest_s16_tower_q14_obj7, questDefinition=&quest_s16_tower_q14}
  sub_nodes [FortConversationRequirement_ControllerRequirement_64 {
    class=FortConversationRequirement_ControllerRequirement
    properties {
     nodeName=Require: Has Quest To Turn In
     requirement {
      class=FortControllerRequirement_HasQuestToTurnIn
      properties {bInverseCondition=true, softQuestToTurnIn=&quest_s16_tower_q14}
     }
    }
   }]
 }
 Speech_C_48 {
  id=#7
  class=Speech_C
  outputs [#41]
  properties {softTaskIcon=&T_UI_NPC_MainQuest, generalConfig {defaultMessage="HA HA HA HA HA!"}}
  sub_nodes [
   ConversationChoiceNode_50 {
    class=ConversationChoiceNode
    properties {defaultChoiceDisplayText=The Spire, choiceTags {gameplayTags [{tagName=Conversation.Choice.Category.MainQuestline}]}}
   }
   FortConversationRequirement_ControllerRequirement_59 {
    class=FortConversationRequirement_ControllerRequirement
    properties {
     nodeName=Require: Has Completed Quest
     requirement {
      class=FortControllerRequirement_HasCompletedQuest
      properties {bInverseCondition=true, questDefinition=&quest_s16_tower_q14}
     }
    }
   }
   FortConversationRequirement_ControllerRequirement_60 {
    class=FortConversationRequirement_ControllerRequirement
    properties {
     nodeName=Require: Is Quest Active
     requirement {class=FortControllerRequirement_IsQuestActive, properties {questDefinition=&quest_s16_tower_q14}}
    }
   }
   FortConversationRequirement_ControllerRequirement_61 {
    class=FortConversationRequirement_ControllerRequirement
    properties {
     nodeName=Require: Has Completed Quest Objective
     requirement {
      class=FortControllerRequirement_HasCompletedQuestObjective
      properties {
       bInverseCondition=true
       objectiveBackendName=quest_s16_tower_q14_obj6
       questDefinition=&quest_s16_tower_q14
      }
     }
    }
   }
   FortConversationRequirement_ControllerRequirement_62 {
    class=FortConversationRequirement_ControllerRequirement
    properties {
     nodeName=Require: NPC Has Conversation Tag
     participantId {tagName=Conversation.Participant.NPC}
     requirement {class=FortControllerRequirement_NPCHasConversationEntryTag, properties {queryTag {tagName=Conversation.Entry.SlurpJonesy}}}
    }
   }
  ]
 }
 Speech_C_12 {
  id=#41
  class=Speech_C
  outputs [#42]
  properties {generalConfig {defaultMessage="JONES IS <b>DOOMED. </>"}}
  sub_nodes [ConversationChoiceNode_43 {class=ConversationChoiceNode, properties {defaultChoiceDisplayText=Continue}}]
 }
 Speech_C_49 {
  id=#42
  class=Speech_C
  outputs [#43]
  properties {generalConfig {defaultMessage="ONLY <b>THE FIRST JONES</> COULD HELP NOW. "}}
  sub_nodes [ConversationChoiceNode_51 {class=ConversationChoiceNode, properties {defaultChoiceDisplayText=Continue}}]
 }
 Speech_C_50 {
  id=#43
  class=Speech_C
  outputs [#44]
  properties {generalConfig {defaultMessage="NOW <b>SLURP OFF!</>"}}
  sub_nodes [ConversationChoiceNode_52 {class=ConversationChoiceNode, properties {defaultChoiceDisplayText=Continue}}]
 }
 FortConversationTaskNode_CompleteQuestObjectiveByBackendName_9 {
  id=#44
  class=FortConversationTaskNode_CompleteQuestObjectiveByBackendName
  properties {objectiveBackendName=quest_s16_tower_q14_obj6, questDefinition=&quest_s16_tower_q14}
  sub_nodes [FortConversationRequirement_ControllerRequirement_63 {
    class=FortConversationRequirement_ControllerRequirement
    properties {
     nodeName=Require: Has Quest To Turn In
     requirement {
      class=FortControllerRequirement_HasQuestToTurnIn
      properties {bInverseCondition=true, softQuestToTurnIn=&quest_s16_tower_q14}
     }
    }
   }]
 }
 Speech_C_8 {
  id=#6
  class=Speech_C
  outputs [#45]
  properties {
   softTaskIcon=&T_UI_NPC_MainQuest
   generalConfig {defaultMessage="You needed a Jones and you came to me? \r\n<b>Look at me!</>"}
  }
  sub_nodes [
   ConversationChoiceNode_13 {
    class=ConversationChoiceNode
    properties {defaultChoiceDisplayText=The Spire, choiceTags {gameplayTags [{tagName=Conversation.Choice.Category.MainQuestline}]}}
   }
   FortConversationRequirement_ControllerRequirement_20 {
    class=FortConversationRequirement_ControllerRequirement
    properties {
     nodeName=Require: Has Completed Quest
     requirement {
      class=FortControllerRequirement_HasCompletedQuest
      properties {bInverseCondition=true, questDefinition=&quest_s16_tower_q14}
     }
    }
   }
   FortConversationRequirement_ControllerRequirement_21 {
    class=FortConversationRequirement_ControllerRequirement
    properties {
     nodeName=Require: Is Quest Active
     requirement {class=FortControllerRequirement_IsQuestActive, properties {questDefinition=&quest_s16_tower_q14}}
    }
   }
   FortConversationRequirement_ControllerRequirement_22 {
    class=FortConversationRequirement_ControllerRequirement
    properties {
     nodeName=Require: Has Completed Quest Objective
     requirement {
      class=FortControllerRequirement_HasCompletedQuestObjective
      properties {
       bInverseCondition=true
       objectiveBackendName=quest_s16_tower_q14_obj4
       questDefinition=&quest_s16_tower_q14
      }
     }
    }
   }
   FortConversationRequirement_ControllerRequirement_23 {
    class=FortConversationRequirement_ControllerRequirement
    properties {
     nodeName=Require: NPC Has Conversation Tag
     participantId {tagName=Conversation.Participant.NPC}
     requirement {class=FortControllerRequirement_NPCHasConversationEntryTag, properties {queryTag {tagName=Conversation.Entry.Rex}}}
    }
   }
  ]
 }
 Speech_C_9 {
  id=#45
  class=Speech_C
  outputs [#46]
  properties {generalConfig {defaultMessage="You know why I'm dressed like a dinosaur? \r\nBecause I <b>sure don't. </>"}}
  sub_nodes [ConversationChoiceNode_39 {class=ConversationChoiceNode, properties {defaultChoiceDisplayText=Continue}}]
 }
 Speech_C_37 {
  id=#46
  class=Speech_C
  outputs [#47, #48]
  properties {generalConfig {defaultMessage="When you find a Jones who knows what's going on, "}}
  sub_nodes [ConversationChoiceNode_55 {class=ConversationChoiceNode, properties {defaultChoiceDisplayText=Continue}}]
 }
 Speech_C_53 {
  id=#48
  class=Speech_C
  outputs [#47]
  properties {generalConfig {defaultMessage="ask about my whole <b>dino situation,</> would ya?"}}
  sub_nodes [ConversationChoiceNode_14 {class=ConversationChoiceNode, properties {defaultChoiceDisplayText=Continue}}]
 }
 FortConversationTaskNode_CompleteQuestObjectiveByBackendName_4 {
  id=#47
  class=FortConversationTaskNode_CompleteQuestObjectiveByBackendName
  properties {objectiveBackendName=quest_s16_tower_q14_obj4, questDefinition=&quest_s16_tower_q14}
  sub_nodes [FortConversationRequirement_ControllerRequirement_56 {
    class=FortConversationRequirement_ControllerRequirement
    properties {
     nodeName=Require: Has Quest To Turn In
     requirement {
      class=FortControllerRequirement_HasQuestToTurnIn
      properties {bInverseCondition=true, softQuestToTurnIn=&quest_s16_tower_q14}
     }
    }
   }]
 }
 Speech_C_4 {
  id=#5
  class=Speech_C
  outputs [#49]
  properties {
   softTaskIcon=&T_UI_NPC_MainQuest
   generalConfig {defaultMessage="Me? Best I can guess is I was on some sort of <b>undercover mission. </>"}
  }
  sub_nodes [
   ConversationChoiceNode_10 {
    class=ConversationChoiceNode
    properties {defaultChoiceDisplayText=The Spire, choiceTags {gameplayTags [{tagName=Conversation.Choice.Category.MainQuestline}]}}
   }
   FortConversationRequirement_ControllerRequirement_16 {
    class=FortConversationRequirement_ControllerRequirement
    properties {
     nodeName=Require: Has Completed Quest
     requirement {
      class=FortControllerRequirement_HasCompletedQuest
      properties {bInverseCondition=true, questDefinition=&quest_s16_tower_q14}
     }
    }
   }
   FortConversationRequirement_ControllerRequirement_17 {
    class=FortConversationRequirement_ControllerRequirement
    properties {
     nodeName=Require: Is Quest Active
     requirement {class=FortControllerRequirement_IsQuestActive, properties {questDefinition=&quest_s16_tower_q14}}
    }
   }
   FortConversationRequirement_ControllerRequirement_18 {
    class=FortConversationRequirement_ControllerRequirement
    properties {
     nodeName=Require: Has Completed Quest Objective
     requirement {
      class=FortControllerRequirement_HasCompletedQuestObjective
      properties {
       bInverseCondition=true
       objectiveBackendName=quest_s16_tower_q14_obj3
       questDefinition=&quest_s16_tower_q14
      }
     }
    }
   }
   FortConversationRequirement_ControllerRequirement_19 {
    class=FortConversationRequirement_ControllerRequirement
    properties {
     nodeName=Require: NPC Has Conversation Tag
     participantId {tagName=Conversation.Participant.NPC}
     requirement {class=FortControllerRequirement_NPCHasConversationEntryTag, properties {queryTag {tagName=Conversation.Entry.GrillSeargent}}}
    }
   }
  ]
 }
 Speech_C_6 {
  id=#49
  class=Speech_C
  outputs [#50]
  properties {generalConfig {defaultMessage="Doesn't matter. Once you're here, you're <b>here. </>"}}
  sub_nodes [ConversationChoiceNode_36 {class=ConversationChoiceNode, properties {defaultChoiceDisplayText=Continue}}]
 }
 Speech_C_34 {
  id=#50
  class=Speech_C
  outputs [#51]
  properties {generalConfig {defaultMessage="And most of us don't even know that we've forgotten... all the stuff we've forgotten. "}}
  sub_nodes [ConversationChoiceNode_37 {class=ConversationChoiceNode, properties {defaultChoiceDisplayText=Continue}}]
 }
 Speech_C_35 {
  id=#51
  class=Speech_C
  outputs [#52]
  properties {generalConfig {defaultMessage="Supposedly it gets worse the more times you enter the <b>loop. </>"}}
  sub_nodes [ConversationChoiceNode_38 {class=ConversationChoiceNode, properties {defaultChoiceDisplayText=Continue}}]
 }
 Speech_C_36 {
  id=#52
  class=Speech_C
  outputs [#53]
  properties {generalConfig {defaultMessage="Can't remember who told me that... "}}
  sub_nodes [ConversationChoiceNode_12 {class=ConversationChoiceNode, properties {defaultChoiceDisplayText=Continue}}]
 }
 FortConversationTaskNode_CompleteQuestObjectiveByBackendName_3 {
  id=#53
  class=FortConversationTaskNode_CompleteQuestObjectiveByBackendName
  properties {objectiveBackendName=quest_s16_tower_q14_obj3, questDefinition=&quest_s16_tower_q14}
  sub_nodes [FortConversationRequirement_ControllerRequirement_55 {
    class=FortConversationRequirement_ControllerRequirement
    properties {
     nodeName=Require: Has Quest To Turn In
     requirement {
      class=FortControllerRequirement_HasQuestToTurnIn
      properties {bInverseCondition=true, softQuestToTurnIn=&quest_s16_tower_q14}
     }
    }
   }]
 }
 Speech_C_0 {
  id=#4
  class=Speech_C
  outputs [#54]
  properties {
   softTaskIcon=&T_UI_NPC_MainQuest
   generalConfig {defaultMessage="What do you mean <b>\"Other Joneses?\" </>\r\nAre they here?"}
  }
  sub_nodes [
   ConversationChoiceNode_5 {
    class=ConversationChoiceNode
    properties {defaultChoiceDisplayText=The Spire, choiceTags {gameplayTags [{tagName=Conversation.Choice.Category.MainQuestline}]}}
   }
   FortConversationRequirement_ControllerRequirement_12 {
    class=FortConversationRequirement_ControllerRequirement
    properties {
     nodeName=Require: Has Completed Quest
     requirement {
      class=FortControllerRequirement_HasCompletedQuest
      properties {bInverseCondition=true, questDefinition=&quest_s16_tower_q14}
     }
    }
   }
   FortConversationRequirement_ControllerRequirement_13 {
    class=FortConversationRequirement_ControllerRequirement
    properties {
     nodeName=Require: Is Quest Active
     requirement {class=FortControllerRequirement_IsQuestActive, properties {questDefinition=&quest_s16_tower_q14}}
    }
   }
   FortConversationRequirement_ControllerRequirement_14 {
    class=FortConversationRequirement_ControllerRequirement
    properties {
     nodeName=Require: Has Completed Quest Objective
     requirement {
      class=FortControllerRequirement_HasCompletedQuestObjective
      properties {
       bInverseCondition=true
       objectiveBackendName=quest_s16_tower_q14_obj2
       questDefinition=&quest_s16_tower_q14
      }
     }
    }
   }
   FortConversationRequirement_ControllerRequirement_15 {
    class=FortConversationRequirement_ControllerRequirement
    properties {
     nodeName=Require: NPC Has Conversation Tag
     participantId {tagName=Conversation.Participant.NPC}
     requirement {class=FortControllerRequirement_NPCHasConversationEntryTag, properties {queryTag {tagName=Conversation.Entry.CastawayJonesy}}}
    }
   }
  ]
 }
 Speech_C_2 {
  id=#54
  class=Speech_C
  outputs [#55]
  properties {generalConfig {defaultMessage=" Do they have <b>food?</>"}}
  sub_nodes [ConversationChoiceNode_6 {class=ConversationChoiceNode, properties {defaultChoiceDisplayText=Continue}}]
 }
 FortConversationTaskNode_CompleteQuestObjectiveByBackendName_2 {
  id=#55
  class=FortConversationTaskNode_CompleteQuestObjectiveByBackendName
  properties {objectiveBackendName=quest_s16_tower_q14_obj2, questDefinition=&quest_s16_tower_q14}
  sub_nodes [FortConversationRequirement_ControllerRequirement_54 {
    class=FortConversationRequirement_ControllerRequirement
    properties {
     nodeName=Require: Has Quest To Turn In
     requirement {
      class=FortControllerRequirement_HasQuestToTurnIn
      properties {bInverseCondition=true, softQuestToTurnIn=&quest_s16_tower_q14}
     }
    }
   }]
 }
 Speech_C_3 {
  id=#3
  class=Speech_C
  outputs [#56]
  properties {
   softTaskIcon=&T_UI_NPC_MainQuest
   generalConfig {defaultMessage="Me? Working with <b>The Foundation?</>\r\nNot likely. "}
  }
  sub_nodes [
   ConversationChoiceNode_3 {
    class=ConversationChoiceNode
    properties {defaultChoiceDisplayText=The Spire, choiceTags {gameplayTags [{tagName=Conversation.Choice.Category.MainQuestline}]}}
   }
   FortConversationRequirement_ControllerRequirement_2 {
    class=FortConversationRequirement_ControllerRequirement
    properties {
     nodeName=Require: Has Completed Quest
     requirement {
      class=FortControllerRequirement_HasCompletedQuest
      properties {bInverseCondition=true, questDefinition=&quest_s16_tower_q14}
     }
    }
   }
   FortConversationRequirement_ControllerRequirement_4 {
    class=FortConversationRequirement_ControllerRequirement
    properties {
     nodeName=Require: Is Quest Active
     requirement {class=FortControllerRequirement_IsQuestActive, properties {questDefinition=&quest_s16_tower_q14}}
    }
   }
   FortConversationRequirement_ControllerRequirement_5 {
    class=FortConversationRequirement_ControllerRequirement
    properties {
     nodeName=Require: Has Completed Quest Objective
     requirement {
      class=FortControllerRequirement_HasCompletedQuestObjective
      properties {
       bInverseCondition=true
       objectiveBackendName=quest_s16_tower_q14_obj1
       questDefinition=&quest_s16_tower_q14
      }
     }
    }
   }
   FortConversationRequirement_ControllerRequirement_6 {
    class=FortConversationRequirement_ControllerRequirement
    properties {
     nodeName=Require: NPC Has Conversation Tag
     participantId {tagName=Conversation.Participant.NPC}
     requirement {class=FortControllerRequirement_NPCHasConversationEntryTag, properties {queryTag {tagName=Conversation.Entry.Cabbie}}}
    }
   }
  ]
 }
 Speech_C_10 {
  id=#56
  class=Speech_C
  outputs [#57]
  properties {generalConfig {defaultMessage="That guy's just itching for an excuse to kick my - "}}
  sub_nodes [ConversationChoiceNode_34 {class=ConversationChoiceNode, properties {defaultChoiceDisplayText=Continue}}]
 }
 Speech_C_32 {
  id=#57
  class=Speech_C
  outputs [#58]
  properties {generalConfig {defaultMessage="Wait - how do <b>you</> know about the Foundation?"}}
  sub_nodes [ConversationChoiceNode_35 {class=ConversationChoiceNode, properties {defaultChoiceDisplayText=Continue}}]
 }
 Speech_C_33 {
  id=#58
  class=Speech_C
  outputs [#59]
  properties {generalConfig {defaultMessage="Also... how do <b>I</> know about The Foundation..."}}
  sub_nodes [ConversationChoiceNode_56 {class=ConversationChoiceNode, properties {defaultChoiceDisplayText=Continue}}]
 }
 Speech_C_54 {
  id=#59
  class=Speech_C
  outputs [#60]
  properties {generalConfig {defaultMessage="<b>The Loop</> does strange things to your memory. "}}
  sub_nodes [ConversationChoiceNode_4 {class=ConversationChoiceNode, properties {defaultChoiceDisplayText=Continue}}]
 }
 FortConversationTaskNode_CompleteQuestObjectiveByBackendName_1 {
  id=#60
  class=FortConversationTaskNode_CompleteQuestObjectiveByBackendName
  properties {objectiveBackendName=quest_s16_tower_q14_obj1, questDefinition=&quest_s16_tower_q14}
  sub_nodes [FortConversationRequirement_ControllerRequirement_53 {
    class=FortConversationRequirement_ControllerRequirement
    properties {
     nodeName=Require: Has Quest To Turn In
     requirement {
      class=FortControllerRequirement_HasQuestToTurnIn
      properties {bInverseCondition=true, softQuestToTurnIn=&quest_s16_tower_q14}
     }
    }
   }]
 }
 Speech_C_1 {
  id=#2
  class=Speech_C
  outputs [#61]
  properties {
   softTaskIcon=&T_UI_NPC_MainQuest
   generalConfig {defaultMessage="Don't you <b>get it?</> \r\nThere is no <b>Agent Jones!</>"}
  }
  sub_nodes [
   ConversationChoiceNode_7 {
    class=ConversationChoiceNode
    properties {defaultChoiceDisplayText=The Spire, choiceTags {gameplayTags [{tagName=Conversation.Choice.Category.MainQuestline}]}}
   }
   FortConversationRequirement_ControllerRequirement_8 {
    class=FortConversationRequirement_ControllerRequirement
    properties {
     nodeName=Require: Has Completed Quest
     requirement {
      class=FortControllerRequirement_HasCompletedQuest
      properties {bInverseCondition=true, questDefinition=&quest_s16_tower_q14}
     }
    }
   }
   FortConversationRequirement_ControllerRequirement_9 {
    class=FortConversationRequirement_ControllerRequirement
    properties {
     nodeName=Require: Is Quest Active
     requirement {class=FortControllerRequirement_IsQuestActive, properties {questDefinition=&quest_s16_tower_q14}}
    }
   }
   FortConversationRequirement_ControllerRequirement_10 {
    class=FortConversationRequirement_ControllerRequirement
    properties {
     nodeName=Require: Has Completed Quest Objective
     requirement {
      class=FortControllerRequirement_HasCompletedQuestObjective
      properties {
       bInverseCondition=true
       objectiveBackendName=quest_s16_tower_q14_obj0
       questDefinition=&quest_s16_tower_q14
      }
     }
    }
   }
   FortConversationRequirement_ControllerRequirement_52 {
    class=FortConversationRequirement_ControllerRequirement
    properties {
     nodeName=Require: NPC Has Conversation Tag
     participantId {tagName=Conversation.Participant.NPC}
     requirement {class=FortControllerRequirement_NPCHasConversationEntryTag, properties {queryTag {tagName=Conversation.Entry.BunkerJonesy}}}
    }
   }
  ]
 }
 Speech_C_7 {
  id=#61
  class=Speech_C
  outputs [#62]
  properties {generalConfig {defaultMessage="Once you're here, you're not <b>you. </>\r\nYou're like a <b>snapshot</> of who you were. "}}
  sub_nodes [ConversationChoiceNode_24 {class=ConversationChoiceNode, properties {defaultChoiceDisplayText=Continue}}]
 }
 Speech_C_30 {
  id=#62
  class=Speech_C
  outputs [#63]
  properties {generalConfig {defaultMessage=I figured that out in the bunker.}}
  sub_nodes [ConversationChoiceNode_33 {class=ConversationChoiceNode, properties {defaultChoiceDisplayText=Continue}}]
 }
 Speech_C_31 {
  id=#63
  class=Speech_C
  outputs [#64]
  properties {generalConfig {defaultMessage="You can <b>figure out</> a lot of stuff in a bunker."}}
  sub_nodes [ConversationChoiceNode_9 {class=ConversationChoiceNode, properties {defaultChoiceDisplayText=Continue}}]
 }
 FortConversationTaskNode_CompleteQuestObjectiveByBackendName_0 {
  id=#64
  class=FortConversationTaskNode_CompleteQuestObjectiveByBackendName
  properties {objectiveBackendName=quest_s16_tower_q14_obj0, questDefinition=&quest_s16_tower_q14}
  sub_nodes [FortConversationRequirement_ControllerRequirement_0 {
    class=FortConversationRequirement_ControllerRequirement
    properties {
     nodeName=Require: Has Quest To Turn In
     requirement {
      class=FortControllerRequirement_HasQuestToTurnIn
      properties {bInverseCondition=true, softQuestToTurnIn=&quest_s16_tower_q14}
     }
    }
   }]
 }
]
